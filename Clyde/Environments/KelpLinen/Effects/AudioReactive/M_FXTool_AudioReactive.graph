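# M_FXTool_AudioReactive is a Default Lit material that appears to be a tool for previewing audio-reactive effects in the KelpLinen environment.
# It remaps the four EventAudio_Spectrum channels through tunable ranges and displays raw values via DebugFloat4Values, tinted by Use_FQ band selectors.
# Looping Time samples an ADSR curve atlas at several rates as simulated pulses.
# Selected pulses scale a normal-based, UV-masked offset; no material output connections are shown.
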Material /Clyde/Environments/KelpLinen/Effects/AudioReactive/M_FXTool_AudioReactive
graph {
 properties {bAutomaticallySetUsageInEditor=false, usedShadingModels=Default Lit}
 nodes [
  CollectionParameter() {collection=/Game/Athena/Events/AudioAnalysis/MPC_Audio_Reactive_Parameters, parameterName=Amplitude_Music}
  n1 = Time()
  CurveAtlasRowParameter(CurveTime=Frac(None=Multiply(A=@n1) {constB=0.6522})) {
   atlas=/ClydeCosmetics/Environments/KelpLinen/CurveAtlas_KelpLinen
   curve=/ClydeCosmetics/Environments/KelpLinen/Curve_KelpLinen_ADSR_AudioReactive
  }
  n2 = Time()
  CurveAtlasRowParameter(CurveTime=Frac(None=Multiply(A=@n2) {constB=1.3043})) {
   atlas=&CurveAtlas_KelpLinen
   curve=&Curve_KelpLinen_ADSR_AudioReactive
  }
  n3 = Time()
  CurveAtlasRowParameter(CurveTime=Frac(None=Multiply(A=@n3) {constB=5.2174})) {
   atlas=&CurveAtlas_KelpLinen
   curve=&Curve_KelpLinen_ADSR_AudioReactive
  }
  n4 = Time()
  n5 = CurveAtlasRowParameter(CurveTime=Frac(None=Multiply(A=@n4) {constB=0.6522})) {
   atlas=&CurveAtlas_KelpLinen
   curve=&Curve_KelpLinen_ADSR_AudioReactive
  }
  LinearInterpolate(A=Constant(), B=@n5, Alpha=ScalarParameter() {parameterName=Use_FQ_1})
  n6 = CollectionParameter() {collection=&FortniteMaterialParameters, parameterName=EventAudio_Spectrum}
  n7 = Multiply(A=@n6)
  n8 = LinearInterpolate(
   A=ComponentMask(None=MaterialFunctionCall!(Vector4=@n7, MaximumNumberOfDigits, UVs, DebugTextLocation RG_UpperRight BA_LowerLeft, Component Spacing) {materialFunction=/Engine/Functions/Engine_MaterialFunctions02/Utility/DebugFloat4Values}.ColorCodedOutput) {b=true, g=true, r=true}
   B=Constant3Vector() {constant=<rgba 1 0 0 0>}
   Alpha=ScalarParameter() {parameterName=Use_FQ_1}
  )
  n9 = Constant3Vector() {constant=<rgba 0 1 0 0>}
  n10 = ScalarParameter() {parameterName=Use_FQ_2}
  LinearInterpolate(
   A=LinearInterpolate(
    A=LinearInterpolate(A=@n8, B=@n9, Alpha=@n10)
    B=Constant3Vector() {constant=<rgba 0 0 1 0>}
    Alpha=ScalarParameter() {parameterName=Use_FQ_3}
   )
   B=Constant3Vector() {constant=<rgba 1 1 1 0>}
   Alpha=ScalarParameter() {parameterName=Use_FQ_4}
  )
  n11 = CollectionParameter() {collection=&FortniteMaterialParameters, parameterName=EventAudio_Spectrum}
  MaterialFunctionCall(
   Input=ComponentMask(None=@n11) {b=true}
   Input Low=ScalarParameter() {defaultValue=0.45, parameterName=3FQ_Input_Low}
   Input High=ScalarParameter() {defaultValue=0.6, parameterName=3FQ_Input_High}
   Target Low=ScalarParameter() {parameterName=3FQ_Output_Low}
   Target High=ScalarParameter() {defaultValue=1, parameterName=3FQ_Output_High}
  ) {materialFunction=&RemapValueRange}
  MaterialFunctionCall(
   Input=ComponentMask(None=@n11) {r=true}
   Input Low=ScalarParameter() {defaultValue=0.7, parameterName=1FQ_Input_Low}
   Input High=ScalarParameter() {defaultValue=0.85, parameterName=1FQ_Input_High}
   Target Low=ScalarParameter() {parameterName=1FQ_Output_Low}
   Target High=ScalarParameter() {defaultValue=1, parameterName=1FQ_Output_High}
  ) {materialFunction=&RemapValueRange}
  MaterialFunctionCall(
   Input=ComponentMask(None=@n11) {g=true}
   Input Low=ScalarParameter() {defaultValue=0.65, parameterName=2FQ_Input_Low}
   Input High=ScalarParameter() {defaultValue=0.8, parameterName=2FQ_Input_High}
   Target Low=ScalarParameter() {parameterName=2FQ_Output_Low}
   Target High=ScalarParameter() {defaultValue=1, parameterName=2FQ_Output_High}
  ) {materialFunction=&RemapValueRange}
  MaterialFunctionCall(
   Input=ComponentMask(None=@n11) {a=true}
   Input Low=ScalarParameter() {defaultValue=0.18, parameterName=4FQ_Input_Low}
   Input High=ScalarParameter() {defaultValue=0.31, parameterName=4FQ_Input_High}
   Target Low=ScalarParameter() {parameterName=4FQ_Output_Low}
   Target High=ScalarParameter() {defaultValue=1, parameterName=4FQ_Output_High}
  ) {materialFunction=&RemapValueRange}
  n12 = AppendVector(A=Constant() {r=1}, B=Constant() {r=1})
  n13 = Time()
  n14 = CurveAtlasRowParameter(CurveTime=Frac(None=Multiply(A=@n13) {constB=0.6522})) {
   atlas=&CurveAtlas_KelpLinen
   curve=&Curve_KelpLinen_ADSR_AudioReactive
  }
  n15 = ScalarParameter() {parameterName=Use_FQ_2}
  n16 = Constant() {r=0.6522}
  n17 = Multiply(A=Time(), B=@n16) {constB=0.6522}
  n18 = Divide(B=@n16)
  n19 = CurveAtlasRowParameter(CurveTime=Frac(None=Add(A=@n17, B=@n18))) {
   atlas=&CurveAtlas_KelpLinen
   curve=&Curve_KelpLinen_ADSR_AudioReactive
  }
  n20 = LinearInterpolate(
   A=LinearInterpolate(A=LinearInterpolate(B=@n14, Alpha=@n15), B=@n19, Alpha=ScalarParameter() {parameterName=Use_FQ_3})
   Alpha=ScalarParameter() {parameterName=Use_FQ_4}
  )
  n21 = LinearInterpolate(Alpha=LinearInterpolate(Alpha=Clamp(None=@n20))) {constB=200}
  n22 = TextureCoordinate()
  Multiply(A=Multiply(A=AppendVector(A=@n12, B=@n21), B=VertexNormalWS()), B=ComponentMask(None=OneMinus(None=@n22)) {g=true})
  PerInstanceRandom()
 ]
}
